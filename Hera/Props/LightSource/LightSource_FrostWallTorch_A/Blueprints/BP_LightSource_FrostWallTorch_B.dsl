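; Frost wall torch light-source Blueprint (variant B) that recolors its flame effects from the FX Color variable during construction.
; After calling the parent script, it destroys the bloom plane if Turn Off is set.
; Otherwise it feeds the color to the Niagara torch effect and creates dynamic materials for the flame and bloom plane.
; Their hue shift is the HSV hue difference from a reference orange, normalized by 360.
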(event Custom|Call%20Animation%20Curve%20Again_1)

(fn ConstructionScript
  (bind _fx_color (Variables|Default|GetFX%20Color))
  (bind (_r _g _b) (Utilities|Struct|BreakLinearColor _fx_color))
  (bind _returnvalue (Math|Color|RGBtoHSV(LinearColor) "(R=1.000000,G=0.250000,B=0.000000,A=1.000000)"))
  (bind _returnvalue_2 (Math|Color|RGBtoHSV(LinearColor) _fx_color))
  (bind _returnvalue_3 (- _returnvalue_2 _returnvalue))
  (bind _r_2 (Utilities|Struct|BreakLinearColor _returnvalue_3))
  (bind _returnvalue_4 (/ _r_2 3.6e+02))
  (|CallParentFunction)
  (if (Variables|Default|GetTurn%20Off)
    (Components|DestroyComponent (Variables|Default|GetBloomPlane))
    (else
      (Niagara|SetNiagaraVariable(LinearColor) (Variables|Default|GetNSLW_Env_Jungle_Torch) "Input Color" (Math|Color|MakeColor _r _g _b))
      (bind _returnvalue_5 (Rendering|Material|CreateDynamicMaterialInstance (Variables|Default|GetFlame) 0 (Rendering|Material|GetMaterial (Variables|Default|GetFlame) 0) "None" :!class /Script/Engine.PrimitiveComponent))
      (Rendering|Material|SetScalarParameterValue _returnvalue_5 "Hue Shift Percentage (S)" _returnvalue_4 :!class /Script/Engine.MaterialInstanceDynamic)
      (bind _returnvalue_6 (Rendering|Material|CreateDynamicMaterialInstance (Variables|Default|GetBloomPlane) 0 (Rendering|Material|GetMaterial (Variables|Default|GetBloomPlane) 0) "None" :!class /Script/Engine.PrimitiveComponent))
      (Rendering|Material|SetScalarParameterValue _returnvalue_6 "Hue Shift Percentage (S)" _returnvalue_4 :!class /Script/Engine.MaterialInstanceDynamic))))
